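# Physics asset for the Stone Viper backpack cosmetic, defining simulated bodies for its dangling moss bones.
# Small capsule bodies sit on a master moss bone and two two-bone chains, one for ear moss and one for moss.
# Angular constraints link each chain segment and root both chains to the master, allowing moderate swing with very tight twist.
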PhysicsAsset /BRCosmetics/Accessories/FORT_Backpacks/Backpack_F_MED_Stone_Viper/Meshes/F_MED_Stone_Viper_Pack_Physics
bodies {
 dyn_ear_moss [{shapeType=Capsule, length=0.505, radius=0.505, center=<1 0 0>, rotation=<rot 90 0 0>}]
 dyn_ear_moss_01 [{shapeType=Capsule, length=0.505, radius=0.505, center=<1 0 0>, rotation=<rot 90 0 0>}]
 dyn_moss [{shapeType=Capsule, length=0.505, radius=0.505, center=<1 0 0>, rotation=<rot 90 0 0>}]
 dyn_moss_01 [{shapeType=Capsule, length=0.505, radius=0.505, center=<1 0 0>, rotation=<rot 90 0 0>}]
 dyn_moss_MSTR [{shapeType=Capsule, length=0.505, radius=0.505}]
}
constraints [
 {bone1Name=dyn_moss_01, bone2Name=dyn_moss, swing1LimitDegrees=45, swing2LimitDegrees=45, twistLimitDegrees=3}
 {bone1Name=dyn_ear_moss_01, bone2Name=dyn_ear_moss, swing1LimitDegrees=45, swing2LimitDegrees=45, twistLimitDegrees=3}
 {bone1Name=dyn_moss, bone2Name=dyn_moss_MSTR, swing1LimitDegrees=45, swing2LimitDegrees=45, twistLimitDegrees=3}
 {bone1Name=dyn_ear_moss, bone2Name=dyn_moss_MSTR, swing1LimitDegrees=45, swing2LimitDegrees=45, twistLimitDegrees=3}
]
